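# Physics asset for secondary motion of the hood and drawstrings on the Broccoli Casual medium male body cosmetic.
# Capsule bodies on spine_04 and spine_05 cover the upper torso, and hood MSTR bodies attach to spine_05 with free angular constraints.
# Hood segments hang from their MSTR bones with tight swing and twist limits.
# Each four-link drawstring chain starts at a MSTR body rotation-locked to its front hood segment, so strings follow the hood.
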PhysicsAsset /BRCosmetics/Characters/Player/Male/Medium/Bodies/M_MED_Broccoli_Casual/Meshes/M_MED_Broccoli_Casual_Dyn
bodies {
 dyn_hood_bk_bott_01 [{shapeType=Capsule, length=5.0573, radius=3.5688, center=<2.6387 0 0>, rotation=<rot 90 0 0>}]
 dyn_hood_bk_bott_02 [{shapeType=Capsule, length=5.0573, radius=3.5688, center=<2.6387 0 0>, rotation=<rot 90 0 0>}]
 dyn_hood_bk_bott_MSTR [{shapeType=Capsule, length=0.505, radius=0.505}]
 dyn_hood_bk_top [{shapeType=Capsule, length=4.3698, radius=3.4751, center=<0.7529 0 0.0002>, rotation=<rot 90 0 0>}]
 dyn_hood_bk_top_MSTR [{shapeType=Capsule, length=0.505, radius=0.505}]
 dyn_hood_fr_top_l [{shapeType=Capsule, length=4.3698, radius=3.4751, center=<0.7529 0 0.0002>, rotation=<rot 90 0 0>}]
 dyn_hood_fr_top_l_MSTR [{shapeType=Capsule, length=0.505, radius=0.505}]
 dyn_hood_fr_top_r [{shapeType=Capsule, length=4.3698, radius=3.4751, center=<0.7529 0 0.0002>, rotation=<rot 90 0 0>}]
 dyn_hood_fr_top_r_MSTR [{shapeType=Capsule, length=0.505, radius=0.505}]
 dyn_hood_l_top [{shapeType=Capsule, length=4.3698, radius=3.4751, center=<0.7529 0 0.0002>, rotation=<rot 90 0 0>}]
 dyn_hood_l_top_MSTR [{shapeType=Capsule, length=0.505, radius=0.505}]
 dyn_hood_r_top [{shapeType=Capsule, length=4.3698, radius=3.4751, center=<0.7529 0 0.0002>, rotation=<rot 90 0 0>}]
 dyn_hood_r_top_MSTR [{shapeType=Capsule, length=0.505, radius=0.505}]
 dyn_string_l_01 [{shapeType=Capsule, length=3.1511, radius=0.6313, center=<2.1117 0 0>, rotation=<rot 90 0 0>}]
 dyn_string_l_02 [{shapeType=Capsule, length=4.3698, radius=0.9751, center=<2.5344 0 0>, rotation=<rot 90 0 0>}]
 dyn_string_l_03 [{shapeType=Capsule, length=4.3698, radius=0.9751, center=<2.5344 0 0>, rotation=<rot 90 0 0>}]
 dyn_string_l_04 [{shapeType=Capsule, length=1.9011, radius=0.9751, center=<1.2993 0.0001 0>, rotation=<rot 90 0 0>}]
 dyn_string_l_MSTR [{shapeType=Capsule, length=0.505, radius=0.505}]
 dyn_string_r_01 [{shapeType=Capsule, length=3.1511, radius=0.6313, center=<2.1117 0 0>, rotation=<rot 90 0 0>}]
 dyn_string_r_02 [{shapeType=Capsule, length=4.3698, radius=0.9751, center=<2.5344 0 0>, rotation=<rot 90 0 0>}]
 dyn_string_r_03 [{shapeType=Capsule, length=4.3698, radius=0.9751, center=<2.5344 0 0>, rotation=<rot 90 0 0>}]
 dyn_string_r_04 [{shapeType=Capsule, length=1.9011, radius=0.9751, center=<1.2993 0.0001 0>, rotation=<rot 90 0 0>}]
 dyn_string_r_MSTR [{shapeType=Capsule, length=0.505, radius=0.505}]
 spine_04 [{shapeType=Capsule, length=4.3538, radius=14.5, center=<2.8662 0.0411 0>}]
 spine_05 [{shapeType=Capsule, length=15.4788, radius=12.3125, center=<-1.4221 -1.0957 0>}]
}
constraints [
 {bone1Name=dyn_hood_fr_top_l, bone2Name=dyn_hood_fr_top_l_MSTR, swing1LimitDegrees=20, swing2LimitDegrees=20, twistLimitDegrees=2}
 {
  bone1Name=dyn_string_l_MSTR
  bone2Name=dyn_hood_fr_top_l
  swing1LimitDegrees=45
  swing1Motion=Locked
  swing2LimitDegrees=45
  swing2Motion=Locked
  twistLimitDegrees=45
  twistMotion=Locked
 }
 {bone1Name=dyn_string_l_01, bone2Name=dyn_string_l_MSTR, swing1LimitDegrees=20, swing2LimitDegrees=20, twistLimitDegrees=2}
 {bone1Name=dyn_string_l_02, bone2Name=dyn_string_l_01, swing1LimitDegrees=20, swing2LimitDegrees=20, twistLimitDegrees=2}
 {bone1Name=dyn_string_l_03, bone2Name=dyn_string_l_02, swing1LimitDegrees=20, swing2LimitDegrees=20, twistLimitDegrees=2}
 {bone1Name=dyn_string_l_04, bone2Name=dyn_string_l_03, swing1LimitDegrees=20, swing2LimitDegrees=20, twistLimitDegrees=2}
 {bone1Name=dyn_hood_l_top, bone2Name=dyn_hood_l_top_MSTR, swing1LimitDegrees=20, swing2LimitDegrees=20, twistLimitDegrees=2}
 {bone1Name=dyn_hood_bk_top, bone2Name=dyn_hood_bk_top_MSTR, swing1LimitDegrees=20, swing2LimitDegrees=20, twistLimitDegrees=2}
 {bone1Name=dyn_hood_bk_bott_01, bone2Name=dyn_hood_bk_bott_MSTR, swing1LimitDegrees=10, swing2LimitDegrees=20, twistLimitDegrees=2}
 {bone1Name=dyn_hood_bk_bott_02, bone2Name=dyn_hood_bk_bott_01, swing1LimitDegrees=10, swing2LimitDegrees=20, twistLimitDegrees=2}
 {bone1Name=dyn_hood_fr_top_r, bone2Name=dyn_hood_fr_top_r_MSTR, swing1LimitDegrees=20, swing2LimitDegrees=20, twistLimitDegrees=2}
 {
  bone1Name=dyn_string_r_MSTR
  bone2Name=dyn_hood_fr_top_r
  swing1LimitDegrees=45
  swing1Motion=Locked
  swing2LimitDegrees=45
  swing2Motion=Locked
  twistLimitDegrees=45
  twistMotion=Locked
 }
 {bone1Name=dyn_string_r_01, bone2Name=dyn_string_r_MSTR, swing1LimitDegrees=20, swing2LimitDegrees=20, twistLimitDegrees=2}
 {bone1Name=dyn_string_r_02, bone2Name=dyn_string_r_01, swing1LimitDegrees=20, swing2LimitDegrees=20, twistLimitDegrees=2}
 {bone1Name=dyn_string_r_03, bone2Name=dyn_string_r_02, swing1LimitDegrees=20, swing2LimitDegrees=20, twistLimitDegrees=2}
 {bone1Name=dyn_string_r_04, bone2Name=dyn_string_r_03, swing1LimitDegrees=20, swing2LimitDegrees=20, twistLimitDegrees=2}
 {bone1Name=dyn_hood_r_top, bone2Name=dyn_hood_r_top_MSTR, swing1LimitDegrees=20, swing2LimitDegrees=20, twistLimitDegrees=2}
 {
  bone1Name=dyn_hood_fr_top_l_MSTR
  bone2Name=spine_05
  swing1LimitDegrees=45
  swing1Motion=Free
  swing2LimitDegrees=45
  swing2Motion=Free
  twistLimitDegrees=45
  twistMotion=Free
 }
 {
  bone1Name=dyn_hood_l_top_MSTR
  bone2Name=spine_05
  swing1LimitDegrees=45
  swing1Motion=Free
  swing2LimitDegrees=45
  swing2Motion=Free
  twistLimitDegrees=45
  twistMotion=Free
 }
 {
  bone1Name=dyn_hood_bk_top_MSTR
  bone2Name=spine_05
  swing1LimitDegrees=45
  swing1Motion=Free
  swing2LimitDegrees=45
  swing2Motion=Free
  twistLimitDegrees=45
  twistMotion=Free
 }
 {
  bone1Name=dyn_hood_bk_bott_MSTR
  bone2Name=spine_05
  swing1LimitDegrees=45
  swing1Motion=Free
  swing2LimitDegrees=45
  swing2Motion=Free
  twistLimitDegrees=45
  twistMotion=Free
 }
 {
  bone1Name=dyn_hood_fr_top_r_MSTR
  bone2Name=spine_05
  swing1LimitDegrees=45
  swing1Motion=Free
  swing2LimitDegrees=45
  swing2Motion=Free
  twistLimitDegrees=45
  twistMotion=Free
 }
 {
  bone1Name=dyn_hood_r_top_MSTR
  bone2Name=spine_05
  swing1LimitDegrees=45
  swing1Motion=Free
  swing2LimitDegrees=45
  swing2Motion=Free
  twistLimitDegrees=45
  twistMotion=Free
 }
 {
  bone1Name=spine_05
  bone2Name=spine_04
  swing1LimitDegrees=45
  swing1Motion=Free
  swing2LimitDegrees=45
  swing2Motion=Free
  twistLimitDegrees=45
  twistMotion=Free
 }
]
