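# Physics asset for the dynamic jacket on the male medium Broccoli Casual cosmetic body, giving jacket bones capsule bodies for physical motion.
# Four chains (front left, front right, left, right) each run from a small MSTR body through two larger segments, with tight twist and asymmetric swing limits.
# Each MSTR body joins spine_02 through an angularly free constraint.
# Pelvis, lower spine and thigh capsules supply body volumes around the jacket, also linked by free constraints.
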PhysicsAsset /BRCosmetics/Characters/Player/Male/Medium/Bodies/M_MED_Broccoli_Casual/Meshes/M_MED_Broccoli_Casual_Dyn_jacket
bodies {
 dyn_jacket_fr_l_01 [{shapeType=Capsule, length=7.0938, radius=5, center=<5.9331 -0.0003 4.5>, rotation=<rot 90 0 0>}]
 dyn_jacket_fr_l_01_MSTR [{shapeType=Capsule, length=1, radius=1}]
 dyn_jacket_fr_l_02 [{shapeType=Capsule, length=7.0938, radius=5, center=<5.9331 -0.0003 4.5>, rotation=<rot 90 0 0>}]
 dyn_jacket_fr_r_01 [{shapeType=Capsule, length=7.0938, radius=5, center=<5.9331 -0.0003 4.5>, rotation=<rot 90 0 0>}]
 dyn_jacket_fr_r_01_MSTR [{shapeType=Capsule, length=1, radius=1}]
 dyn_jacket_fr_r_02 [{shapeType=Capsule, length=7.0938, radius=5, center=<5.9331 -0.0003 4.5>, rotation=<rot 90 0 0>}]
 dyn_jacket_l_01 [{shapeType=Capsule, length=7.0938, radius=5, center=<5.9331 -0.0003 4.5>, rotation=<rot 90 0 0>}]
 dyn_jacket_l_01_MSTR [{shapeType=Capsule, length=1, radius=1}]
 dyn_jacket_l_02 [{shapeType=Capsule, length=7.0938, radius=5, center=<5.9331 -0.0003 4.5>, rotation=<rot 90 0 0>}]
 dyn_jacket_r_01 [{shapeType=Capsule, length=7.0938, radius=5, center=<5.9331 -0.0003 4.5>, rotation=<rot 90 0 0>}]
 dyn_jacket_r_01_MSTR [{shapeType=Capsule, length=1, radius=1}]
 dyn_jacket_r_02 [{shapeType=Capsule, length=7.0938, radius=5, center=<5.9331 -0.0003 4.5>, rotation=<rot 90 0 0>}]
 pelvis [{shapeType=Capsule, length=8.9273, radius=11.8046, center=<3.2501 0.1632 -0.0001>}]
 spine_01 [{shapeType=Capsule, length=7.2812, radius=10.25, center=<3.2501 0.7666 -0.0001>}]
 spine_02 [{shapeType=Capsule, length=7.7188, radius=9.8125, center=<3.2501 1.4404 -0.0002>}]
 thigh_l [{shapeType=Capsule, length=29.9062, radius=8.6562, center=<-17.2119 0.0003 0>, rotation=<rot 90 0 0>}]
 thigh_r [{shapeType=Capsule, length=29.9062, radius=8.6562, center=<17.2119 -0.0003 0>, rotation=<rot 90 0 180>}]
}
constraints [
 {bone1Name=dyn_jacket_fr_l_02, bone2Name=dyn_jacket_fr_l_01, swing1LimitDegrees=20, swing2LimitDegrees=45, twistLimitDegrees=5}
 {bone1Name=dyn_jacket_l_02, bone2Name=dyn_jacket_l_01, swing1LimitDegrees=10, swing2LimitDegrees=45, twistLimitDegrees=5}
 {bone1Name=dyn_jacket_r_02, bone2Name=dyn_jacket_r_01, swing1LimitDegrees=10, swing2LimitDegrees=45, twistLimitDegrees=5}
 {bone1Name=dyn_jacket_fr_r_02, bone2Name=dyn_jacket_fr_r_01, swing1LimitDegrees=20, swing2LimitDegrees=45, twistLimitDegrees=5}
 {bone1Name=dyn_jacket_fr_l_01, bone2Name=dyn_jacket_fr_l_01_MSTR, swing1LimitDegrees=10, swing2LimitDegrees=45, twistLimitDegrees=5}
 {bone1Name=dyn_jacket_l_01, bone2Name=dyn_jacket_l_01_MSTR, swing1LimitDegrees=5, swing2LimitDegrees=45, twistLimitDegrees=5}
 {bone1Name=dyn_jacket_r_01, bone2Name=dyn_jacket_r_01_MSTR, swing1LimitDegrees=5, swing2LimitDegrees=45, twistLimitDegrees=5}
 {bone1Name=dyn_jacket_fr_r_01, bone2Name=dyn_jacket_fr_r_01_MSTR, swing1LimitDegrees=10, swing2LimitDegrees=45, twistLimitDegrees=5}
 {
  bone1Name=thigh_l
  bone2Name=pelvis
  swing1LimitDegrees=45
  swing1Motion=Free
  swing2LimitDegrees=45
  swing2Motion=Free
  twistLimitDegrees=45
  twistMotion=Free
 }
 {
  bone1Name=dyn_jacket_fr_r_01_MSTR
  bone2Name=spine_02
  swing1LimitDegrees=45
  swing1Motion=Free
  swing2LimitDegrees=45
  swing2Motion=Free
  twistLimitDegrees=45
  twistMotion=Free
 }
 {
  bone1Name=dyn_jacket_r_01_MSTR
  bone2Name=spine_02
  swing1LimitDegrees=45
  swing1Motion=Free
  swing2LimitDegrees=45
  swing2Motion=Free
  twistLimitDegrees=45
  twistMotion=Free
 }
 {
  bone1Name=dyn_jacket_fr_l_01_MSTR
  bone2Name=spine_02
  swing1LimitDegrees=45
  swing1Motion=Free
  swing2LimitDegrees=45
  swing2Motion=Free
  twistLimitDegrees=45
  twistMotion=Free
 }
 {
  bone1Name=dyn_jacket_l_01_MSTR
  bone2Name=spine_02
  swing1LimitDegrees=45
  swing1Motion=Free
  swing2LimitDegrees=45
  swing2Motion=Free
  twistLimitDegrees=45
  twistMotion=Free
 }
 {
  bone1Name=spine_01
  bone2Name=pelvis
  swing1LimitDegrees=45
  swing1Motion=Free
  swing2LimitDegrees=45
  swing2Motion=Free
  twistLimitDegrees=45
  twistMotion=Free
 }
 {
  bone1Name=spine_02
  bone2Name=spine_01
  swing1LimitDegrees=45
  swing1Motion=Free
  swing2LimitDegrees=45
  swing2Motion=Free
  twistLimitDegrees=45
  twistMotion=Free
 }
 {
  bone1Name=thigh_r
  bone2Name=pelvis
  swing1LimitDegrees=45
  swing1Motion=Free
  swing2LimitDegrees=45
  swing2Motion=Free
  twistLimitDegrees=45
  twistMotion=Free
 }
]
