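# Physics asset for the Broccoli Casual medium male player body, named as the cloth collision set for that cosmetic mesh.
# The pelvis body carries a single capsule, while the left and right thigh bodies list no shapes.
# Two constraints join each thigh to the pelvis with equal swing and twist limit angles.
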PhysicsAsset /BRCosmetics/Characters/Player/Male/Medium/Bodies/M_MED_Broccoli_Casual/Meshes/M_MED_Broccoli_Casual_ClothCollisions
bodies {pelvis [{shapeType=Capsule, length=4.2812, radius=10.8411, center=<0 -0.5467 0>}], thigh_l [], thigh_r []}
constraints [
 {bone1Name=thigh_l, bone2Name=pelvis, swing1LimitDegrees=45, swing2LimitDegrees=45, twistLimitDegrees=45}
 {bone1Name=thigh_r, bone2Name=pelvis, swing1LimitDegrees=45, swing2LimitDegrees=45, twistLimitDegrees=45}
]
